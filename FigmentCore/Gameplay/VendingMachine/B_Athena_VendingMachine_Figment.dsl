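; Figment-mode vending machine Blueprint that sets its interaction speed and removes itself when disabled by data.
; On BeginPlay it waits for playlist data, then applies the interaction speed from a scalable float.
; A curve table "Enabled" row decides whether it shows its quest minimap icon and runs parent BeginPlay, or hides the quest mesh and destroys itself.
; Overlap, tick and markable events are empty, and damage impulse handling returns false.
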(event EventBeginPlay
  (bind (_value _curve) (Utilities|Struct|BreakScalableFloat (Variables|Default|GetFigmentInteractionSpeed)))
  (Class|FortAsyncActionCurrentPlaylistReady|PlaylistDataReadyAsync
    (:OnReady
      (Variables|Interaction|SetInteractionSpeed _curve)
      (bind _outxy (Fortnite|EvaluateCurveTableRow (Variables|Default|GetEnabled%20Row)))
      (if (> _outxy 0)
        (MiniMapData|SetVisibility (Variables|Default|GetFortQuestIcon) true)
        (|CallParentFunction)
        (else
          (Rendering|SetVisibility (Variables|Default|GetSM_VendingMachineQuest) false false)
          (Actor|DestroyActor))))))

(event Collision|EventActorBeginOverlap
  (params (OtherActor :type Actor%20Object%20Reference)))

(event EventTick
  (params (DeltaSeconds :type Float%20%28single-precision%29)))

(event MarkableActor|EventSetCanBeMarked
  (params (bInCanBeMarked :type Boolean)))

(fn GetMarkerDetailsId
  (return "None"))

(fn HandleCustomImpulseFromDamage
  (params (ImpulseValue :type Vector :default "0, 0, 0") (Damage :type Float%20%28single-precision%29) (DamageTags :type Gameplay%20Tag%20Container%20Structure%20%28by%20ref%29) (HitResult :type Hit%20Result%20Structure%20%28by%20ref%29) (InstigatingController :type Controller%20Object%20Reference) (DamageCauser :type Actor%20Object%20Reference) (EffectContext :type Gameplay%20Effect%20Context%20Handle%20Structure%20%28by%20ref%29))
  (return false))

(fn ConstructionScript
  (|CallParentFunction))
